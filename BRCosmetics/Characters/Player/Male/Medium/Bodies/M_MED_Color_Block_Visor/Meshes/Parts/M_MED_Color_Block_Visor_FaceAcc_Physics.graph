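# Physics asset for the face accessory mesh of the male medium Color Block Visor cosmetic body, defining physics for its dyn_ bones.
# Seven small spheres cover the front, side, middle and back dyn_ bones, alongside one large sphere on the head.
# Each dyn_ body is constrained directly to the head rather than chained, with limited swing and very narrow twist.
# The front bone has a tighter swing limit than the others.
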PhysicsAsset /BRCosmetics/Characters/Player/Male/Medium/Bodies/M_MED_Color_Block_Visor/Meshes/Parts/M_MED_Color_Block_Visor_FaceAcc_Physics
bodies {
 dyn_backleft [{shapeType=Sphere, radius=2.005, shapeName=dyn_backleft_sphere, center=<1.4424 0 0>}]
 dyn_backright [{shapeType=Sphere, radius=1.505, shapeName=dyn_backright_sphere, center=<1.3752 0 0.0001>}]
 dyn_front [{shapeType=Sphere, radius=1.38, shapeName=dyn_front_sphere, center=<1.0936 0 0.0001>}]
 dyn_left [{shapeType=Sphere, radius=1.505, shapeName=dyn_left_sphere, center=<1.3752 0 0.0001>}]
 dyn_mid [{shapeType=Sphere, radius=1.505, shapeName=dyn_mid_sphere, center=<1.3752 0 0.0001>}]
 dyn_midleft [{shapeType=Sphere, radius=1.63, shapeName=dyn_midleft_sphere, center=<1.3897 0 0.0001>}]
 dyn_right [{shapeType=Sphere, radius=1.505, shapeName=dyn_right_sphere, center=<1.3752 0 0.0001>}]
 head [{shapeType=Sphere, radius=7.5098, shapeName=head_sphere, center=<5.4755 2.7803 -0.1024>}]
}
constraints [
 {bone1Name=dyn_front, bone2Name=head, swing1LimitDegrees=30, swing2LimitDegrees=30, twistLimitDegrees=5}
 {bone1Name=dyn_right, bone2Name=head, swing1LimitDegrees=45, swing2LimitDegrees=45, twistLimitDegrees=5}
 {bone1Name=dyn_mid, bone2Name=head, swing1LimitDegrees=45, swing2LimitDegrees=45, twistLimitDegrees=5}
 {bone1Name=dyn_midleft, bone2Name=head, swing1LimitDegrees=45, swing2LimitDegrees=45, twistLimitDegrees=5}
 {bone1Name=dyn_backright, bone2Name=head, swing1LimitDegrees=45, swing2LimitDegrees=45, twistLimitDegrees=5}
 {bone1Name=dyn_left, bone2Name=head, swing1LimitDegrees=45, swing2LimitDegrees=45, twistLimitDegrees=5}
 {bone1Name=dyn_backleft, bone2Name=head, swing1LimitDegrees=45, swing2LimitDegrees=45, twistLimitDegrees=5}
]
