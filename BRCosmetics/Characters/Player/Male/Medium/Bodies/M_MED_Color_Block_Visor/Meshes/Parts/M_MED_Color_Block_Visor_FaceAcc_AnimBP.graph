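# Animation Blueprint for the Color Block Visor character's face accessory, driving its pose and its walkman and cord dynamics.
# It copies the attached parent's pose, blends in walkman and cord bones copied from the body mesh found when customization loads, and caches the result.
# Two rigid body nodes with separate physics assets simulate in neck and root bone space.
# When rigid bodies are disabled, RBAN fades in a fixed transform on the cord master bone.
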AnimBlueprint /BRCosmetics/Characters/Player/Male/Medium/Bodies/M_MED_Color_Block_Visor/Meshes/Parts/M_MED_Color_Block_Visor_FaceAcc_AnimBP
target_skeleton=&/M_MED_Color_Block_Visor_FaceAcc_Skeleton
graphs [
 AnimGraph {nodes [
   n1 = RigidBody(ComponentPose=Local To Component(LocalPose=Use cached pose 'Start Pose'())) {
    bOverrideWorldGravity=true
    overridePhysicsAsset=&/M_MED_Color_Block_Visor_FaceAcc_Physics
    simulationSpace=BaseBoneSpace
    alphaBoolBlend {blendOption=Linear}
    baseBoneRef {boneName=neck_01}
    componentLinearAccScale=<0.2 0.2 0.25>
    componentLinearVelScale=<0.2 0.2 0.25>
    overrideWorldGravity {z=-420}
   }
   n2 = RigidBody(ComponentPose=@n1) {
    bOverrideWorldGravity=true
    overridePhysicsAsset=&/M_MED_Color_Block_Visor_FaceAcc_CordPh
    simulationSpace=BaseBoneSpace
    alphaBoolBlend {blendOption=Linear}
    baseBoneRef {boneName=root}
    componentLinearAccScale=<0.3 0.3 0.3>
    componentLinearVelScale=<0.2 0.2 0.25>
    overrideWorldGravity {z=-980}
   }
   n3 = "Transform (Modify) Bone - Bone: dyn_cord_MSTR"(ComponentPose=@n2, Alpha=Get RBAN() -> (RBAN:double), Translation="0,0,0", Rotation="0,85,0", Scale="1,1,1") {
    rotationMode=BMM_Replace
    rotationSpace=BCS_BoneSpace
    scaleMode=BMM_Replace
    scaleSpace=BCS_BoneSpace
    translationMode=BMM_Replace
    translationSpace=BCS_BoneSpace
    alphaBoolBlend {blendOption=Linear}
    boneToModify {boneName=dyn_cord_MSTR}
    rotation {yaw=85}
   }
   Output Pose(Result=Component To Local(ComponentPose=@n3)) {layerGroup=None}
   n4 = Layered blend per bone(
    BasePose=Copy Pose From Mesh() {bCopyCurves=true, bUseAttachedParent=true, rootBoneToCopy=root}
    BlendPoses_0=Copy Pose From Mesh(SourceMeshComponent=Get Body Mesh() -> (Body Mesh:Skeletal Mesh Component Object Reference)) {bCopyCurves=true, bUseAttachedParent=true, rootBoneToCopy=root}
   ) {layerSetup [{branchFilters [{boneName=dyn_walkman_MSTR}, {boneName=dyn_cord_MSTR}, {boneName=dyn_walkman_01}]}]}
   Save cached pose 'Start Pose'(Pose=@n4) {cacheName=Start Pose}
  ]}
 EventGraph {nodes [
   Event BlueprintUpdateAnimation()
   > n1 = Branch(Condition=Get bRigidBodyEnabled())
   > Set RBAN(RBAN:double=0) -> (Output_Get:double)
   Set RBAN(execute=@n1.else, RBAN=1)
   Event OnCharacterCustomizationLoaded()
   > n2 = Cast To FortPlayerPawnAthena(Object=TryGetPawnOwner())
   > Set Body Mesh(Body Mesh:Skeletal Mesh Component Object Reference=GetSkeletalMeshForPartType(self=@n2)) -> (Output_Get:Skeletal Mesh Component Object Reference)
  ]}
]
variables [
 Body Mesh {type=SkeletalMeshComponent}
 RBAN {type=number}
]
